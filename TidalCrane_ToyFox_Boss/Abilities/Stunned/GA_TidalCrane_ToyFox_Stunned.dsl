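; Gameplay ability that plays the Tidal Crane ToyFox boss's stunned sequence.
; It starts the stun montage, clears old recovery effects, applies the active stun effect and disables look-at turning.
; When that effect is removed, recovery regen and the montage Outro follow, unless a DoNotRemove tag holds the stun for a finisher.
; On end, it applies a recently-stunned effect and restores turning.
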(event Ability|EventActivateAbility
  ;[0 0.15 0.006968 0.5] ActivateAbility: Play out the stunned sequence.
  (|CallParentFunction)
  (Ability|Tasks|PlayMontageAndWaitNotify "PlayMontageTask" (Variables|Settings|GetStunnedAnimMontage)
    (:then
      (GameplayEffects|RemoveActiveGameplayEffectbySourceEffect (Ability|GetAbilitySystemComponentfromActorInfo) (Variables|Settings|GetStunnedRecoveryGameplayEffectClass) 0 -1) ; Remove any stunned recovery GE that may be running in case \nwe are sent back to stun right after recovery.
      ;;[0.02969 0.1111 0.15 0.5] Apply the stunned active gameplay effect and wait for it to be removed.
      (bind _returnvalue (Ability|ApplyGameplayEffectToOwner (Variables|Settings|GetStunnedActiveGameplayEffectClass)))
      (Ability|Tasks|WaitForGameplayEffectRemoved _returnvalue
        (:then
          ;;[0.06018 0 0.15 0.5] Disable turning around when stunned.
          (bind _asbga_tidal_crane_boss_toy_fox (Utilities|Casting|CastToBGA_TidalCrane_Boss_ToyFox (Ability|GetAvatarActorfromActorInfo)))
          (Class|BGATidalCraneBoss|AllowLookatTarget _asbga_tidal_crane_boss_toy_fox false))
        (:OnRemoved
          ;;[0.1398 0.04994 0.15 0.5] Do not remove the stun if there's a DoNotRemove stun tag on the actor. This is to prevent recovery when the finisher is being executed.
          (bind _returnvalue_2 (GameplayTags|HasMatchingGameplayTag(Message) (Ability|GetAvatarActorfromActorInfo) "(TagName=\"Ability.TIdalCrane.ToyFox.Stunned.DoNotRemove\")"))
          (if (not _returnvalue_2)
            ;;[0 0.15 0.1454 0.5] Apply quick health regen.
            (Ability|ApplyGameplayEffectToOwner (Variables|Settings|GetStunnedRecoveryGameplayEffectClass))
            (Ability|Animation|MontageJumptoSection "Outro")))))
    (:OnCompleted
      _)
    (:OnBlendOut
      _)
    (:OnInterrupted
      _)
    (:OnCancelled
      _))
  (Ability|EndAbility))

(event Ability|EventOnEndAbility
  (params (bWasCancelled :type Boolean))
  ;[0.1042 0.05364 0 0.5] OnEndAbility: Clean up.
  (Utilities|FlowControl|Sequence
    (:then_0
      (Ability|RemoveGrantedbyEffect)
      (Ability|ApplyGameplayEffectToOwner (Variables|Settings|GetRecentlyStunnedGameplayEffectClass))
      (bind _asbga_tidal_crane_boss_toy_fox (Utilities|Casting|CastToBGA_TidalCrane_Boss_ToyFox (Ability|GetAvatarActorfromActorInfo)))
      (Class|BGATidalCraneBoss|AllowLookatTarget _asbga_tidal_crane_boss_toy_fox true))
    (:then_1
      (|CallParentFunction bWasCancelled))))
